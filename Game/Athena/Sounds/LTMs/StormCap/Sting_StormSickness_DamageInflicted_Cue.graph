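# Sound cue for the storm sickness "damage inflicted" sting in the StormCap limited-time mode.
# A single wave player feeds the Sting_StormSickness_Inflicted wave straight to the output at reduced volume through an important in-game mission sound class.
# A concurrency override allows one instance at a time and blocks new triggers while it plays.
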SoundCue /Game/Athena/Sounds/LTMs/StormCap/Sting_StormSickness_DamageInflicted_Cue
root=WavePlayer_0
nodes [{id=WavePlayer_0, properties {soundWaveAssetPtr=&/Sting_StormSickness_Inflicted}}]
properties {
 bOverrideConcurrency=true
 maxDistance=2097152
 soundClassObject=&Fort_Mission_Ingame_Important_PSM
 volumeMultiplier=0.5
 concurrencyOverrides {maxCount=1, resolutionRule=PreventNew}
}
