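# Sound cue for the storm sickness warning sting in the StormCap limited-time mode.
# A single wave player feeds the Sting_StormSickness_Warning wave straight to the output at half volume, using an important in-game mission sound class.
# A concurrency override allows one instance at a time and blocks new plays while one is active, with a very large max distance.
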SoundCue /Game/Athena/Sounds/LTMs/StormCap/Sting_StormSickness_Warning_Cue
root=WavePlayer_1
nodes [{id=WavePlayer_1, properties {soundWaveAssetPtr=&/Sting_StormSickness_Warning}}]
properties {
 bOverrideConcurrency=true
 maxDistance=2097152
 soundClassObject=&Fort_Mission_Ingame_Important_PSM
 volumeMultiplier=0.5
 concurrencyOverrides {maxCount=1, resolutionRule=PreventNew}
}
