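# HUD widget that alerts players to storm sickness with warning, infliction and timed recovery messages.
# It follows the view-target pawn, using async gameplay event and InsideSafeZone/StormSickness tag listeners to pick the state and feed a collapsed progress bar.
# Each message sets the texts and plays the Intro animation, with stings for warning and infliction.
# A data-registry scalable float and live-broadcast controller check gate display; pawn death hides it.
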WidgetBlueprint /StormSickness/UI/AthenaStormSicknessWarning
animations [
 StormDMGAlert {
  binding_count=4
  display_label=""
  length_seconds=2.5
  object=&self:StormDMGAlert
  sequence {bindings [
    HelpText {
     id=#1
     display_name=Help_RichText
     possessed_class=CommonTextBlock
     tracks [{
       type=MovieScene2DTransformTrack
       sections [{
         type=MovieScene2DTransformSection
         end_frame=50
         start_frame=0
         channels [
          Translation.X {type=MovieSceneScriptingFloatChannel, const [0,1]}
          Translation.Y {type=MovieSceneScriptingFloatChannel, const [0,1]}
          Angle {type=MovieSceneScriptingFloatChannel, const [0,1]}
          Scale.X {type=MovieSceneScriptingFloatChannel, keys [[0,0],[13,0],[19,1],[22,0.9],[25,1]]}
          Scale.Y {type=MovieSceneScriptingFloatChannel, keys [[0,0],[13,0],[19,1],[22,0.9],[25,1]]}
          Shear.X {type=MovieSceneScriptingFloatChannel, const [0,1]}
          Shear.Y {type=MovieSceneScriptingFloatChannel, const [0,1]}
         ]
        }]
      }]
    }
    AlertText {
     id=#2
     possessed_class=CommonTextBlock
     tracks [{
       type=MovieScene2DTransformTrack
       sections [{
         type=MovieScene2DTransformSection
         end_frame=50
         start_frame=0
         channels [
          Translation.X {type=MovieSceneScriptingFloatChannel, const [0,1]}
          Translation.Y {type=MovieSceneScriptingFloatChannel, const [0,1]}
          Angle {type=MovieSceneScriptingFloatChannel, const [0,1]}
          Scale.X {type=MovieSceneScriptingFloatChannel, keys [[9,0],[13,1],[16,0.9],[19,1]]}
          Scale.Y {type=MovieSceneScriptingFloatChannel, keys [[9,0],[13,1],[16,0.9],[19,1]]}
          Shear.X {type=MovieSceneScriptingFloatChannel, const [0,1]}
          Shear.Y {type=MovieSceneScriptingFloatChannel, const [0,1]}
         ]
        }]
      }]
    }
    StormIcon {
     id=#3
     possessed_class=Image
     tracks [{
       type=MovieScene2DTransformTrack
       sections [{
         type=MovieScene2DTransformSection
         end_frame=49
         start_frame=0
         channels [
          Translation.X {type=MovieSceneScriptingFloatChannel, const [0,1]}
          Translation.Y {type=MovieSceneScriptingFloatChannel, const [0,1]}
          Angle {type=MovieSceneScriptingFloatChannel, const [0,1]}
          Scale.X {type=MovieSceneScriptingFloatChannel, keys [[0,0],[5,1],[7,0.9],[9,1]]}
          Scale.Y {type=MovieSceneScriptingFloatChannel, keys [[0,0],[5,1],[7,0.9],[9,1]]}
          Shear.X {type=MovieSceneScriptingFloatChannel, const [0,1]}
          Shear.Y {type=MovieSceneScriptingFloatChannel, const [0,1]}
         ]
        }]
      }]
    }
    LightningBack {
     id=#4
     possessed_class=Image
     tracks [{
       type=MovieScene2DTransformTrack
       sections [{
         type=MovieScene2DTransformSection
         end_frame=49
         start_frame=0
         channels [
          Translation.X {type=MovieSceneScriptingFloatChannel, const [0,1]}
          Translation.Y {type=MovieSceneScriptingFloatChannel, const [0,1]}
          Angle {type=MovieSceneScriptingFloatChannel, const [0,1]}
          Scale.X {type=MovieSceneScriptingFloatChannel, keys [[9,0],[16,0],[21,1]]}
          Scale.Y {type=MovieSceneScriptingFloatChannel, const [0.9,1]}
          Shear.X {type=MovieSceneScriptingFloatChannel, const [0,1]}
          Shear.Y {type=MovieSceneScriptingFloatChannel, const [0,1]}
         ]
        }]
      }]
    }
   ]}
 }
 Intro {
  binding_count=4
  length_seconds=1
  object=&self:Intro
  sequence {
   bindings [
    Help_RichText {
     id=#5
     possessed_class=CommonRichTextBlock
     tracks [
      {
       type=MovieSceneFloatTrack
       display_name=Render Opacity
       sections [{type=MovieSceneFloatSection, channels [None {type=MovieSceneScriptingFloatChannel, keys [[9,0],[15,1]]}]}]
      }
      {
       type=MovieScene2DTransformTrack
       sections [{type=MovieScene2DTransformSection, channels [Translation.X {type=MovieSceneScriptingFloatChannel, keys [[9,-64],[21,0]]}]}]
      }
     ]
    }
    AlertText {
     id=#6
     possessed_class=CommonTextBlock
     tracks [
      {
       type=MovieSceneFloatTrack
       display_name=Render Opacity
       sections [{type=MovieSceneFloatSection, channels [None {type=MovieSceneScriptingFloatChannel, keys [[6,0],[12,1]]}]}]
      }
      {
       type=MovieScene2DTransformTrack
       sections [{type=MovieScene2DTransformSection, channels [Translation.X {type=MovieSceneScriptingFloatChannel, keys [[6,-64],[18,0]]}]}]
      }
     ]
    }
    LightningBack {
     id=#7
     possessed_class=Image
     tracks [
      {
       type=MovieSceneFloatTrack
       display_name=Render Opacity
       sections [{type=MovieSceneFloatSection, channels [None {type=MovieSceneScriptingFloatChannel, keys [[0,0],[2,1],[15,1],[60,0]]}]}]
      }
      {
       type=MovieScene2DTransformTrack
       sections [{
         type=MovieScene2DTransformSection
         channels [
          Translation.X {type=MovieSceneScriptingFloatChannel, keys [[0,-256],[12,-48]]}
          Angle {type=MovieSceneScriptingFloatChannel, const [0,1]}
          Scale.X {type=MovieSceneScriptingFloatChannel, keys [[2,2],[12,1]]}
          Scale.Y {type=MovieSceneScriptingFloatChannel, const [1.25,1]}
         ]
        }]
      }
     ]
    }
    Overlay_0 {
     id=#8
     possessed_class=Overlay
     tracks [
      {
       type=MovieSceneFloatTrack
       display_name=Render Opacity
       sections [{type=MovieSceneFloatSection, channels [None {type=MovieSceneScriptingFloatChannel, keys [[0,0],[9,1]]}]}]
      }
      {
       type=MovieScene2DTransformTrack
       sections [{
         type=MovieScene2DTransformSection
         channels [
          Translation.X {type=MovieSceneScriptingFloatChannel, const [0,1]}
          Translation.Y {type=MovieSceneScriptingFloatChannel, const [0,1]}
          Angle {type=MovieSceneScriptingFloatChannel, const [0,1]}
          Scale.X {type=MovieSceneScriptingFloatChannel, keys [[0,3],[9,0.9],[15,1]]}
          Scale.Y {type=MovieSceneScriptingFloatChannel, keys [[0,3],[9,0.9],[15,1]]}
          Shear.X {type=MovieSceneScriptingFloatChannel, const [0,1]}
          Shear.Y {type=MovieSceneScriptingFloatChannel, const [0,1]}
         ]
        }]
      }
     ]
    }
   ]
   display_rate {numerator=60}
  }
 }
]
graphs [
 Anim_Load {nodes [
   Anim_Load() {is=title K2Node_FunctionEntry}
   > PlayAnimation(InAnimation=Get Intro())
  ]}
 BuildDataRegistryResolverScope {nodes [
   BuildDataRegistryResolverScope() {is=title K2Node_FunctionEntry}
   n1 = GetOwningPlayer()
   > Return Node!(
    execute
    InOutResolverScopes:Array of Names=Make Array("[0]":Name=GetSpatialGameplayInterfaceFNameForActor!(Actor:Actor Object Reference=@n1) -> (ReturnValue:Name=None)) -> (Array:Array of Names)
    InOutPriority:int=1
    ReturnValue:bool=true
   )
  ]}
 Check Set Cached Pawn {nodes [
   n1 = Check Set Cached Pawn!() -> (then, New Pawn:Fort Player Pawn Athena Object Reference) {is=title K2Node_FunctionEntry}
   > n2 = Branch(Condition="Equal (Object)"(A=@n1, B=Get CachedPawnToUse() -> (CachedPawnToUse:Fort Player Pawn Athena Object Reference)))
   Set CachedPawnToUse(execute=@n2.else, CachedPawnToUse:Fort Player Pawn Athena Object Reference=@n1) -> (Output_Get:Fort Player Pawn Athena Object Reference)
   > SetBinds()
  ]}
 Check Show Or Hide Widget {nodes [
   Inputs!() -> (InExec:Exec)
   > n1 = Branch(Condition=GetValueAsBoolWithContext!(Input:"Scalable Float Structure (by ref)"=Get bStormSicknessEnabled() -> (bStormSicknessEnabled:Scalable Float Structure), Level:float=0) -> (ReturnValue:bool))
   > n2 = Cast To FortLiveBroadcastController!(execute, Object=GetOwningPlayer()) -> (then, CastFailed, AsFort Live Broadcast Controller:Fort Live Broadcast Controller Object Reference) {is=title K2Node_DynamicCast}
   Outputs!(Show:Exec=@n2.CastFailed, False:Exec=[@n2.then, @n1.else])
  ]}
 EventGraph {nodes [
   StopDisplayUI() {is=title K2Node_CustomEvent}
   > SetVisibility(InVisibility=Collapsed)
   > SetHasActiveContent!(execute, self:Fort Prioritized Widget Interface, bNewActive:bool=false) -> (then)
   > Set bShowingInflicted(bShowingInflicted:bool=false) -> (Output_Get:bool)
   Start Storm Sickness Infliction() {is=title K2Node_CustomEvent}
   > n1 = Check Show or Hide Widget!(InExec:Exec) -> (Show:Exec, False:Exec) {is=title K2Node_MacroInstance}
   Clear and Invalidate Timer by Handle(execute=@n1.False, Handle=Get HideTimerHandle() -> (HideTimerHandle:Timer Handle Structure))
   > StopDisplayUI() {is=title K2Node_CallFunction}
   n2 = Branch(
    execute=@n1.Show
    Condition=HasMatchingGameplayTag!(self:Gameplay Tag Asset Interface Interface=Get CachedPawnToUse(), TagToCheck:Gameplay Tag Structure=(TagName="Gameplay.InsideSafeZone")) -> (ReturnValue:bool)
   )
   Set bShowingInflicted(execute=@n2.else, bShowingInflicted=true)
   > Clear and Invalidate Timer by Handle(Handle=Get HideTimerHandle())
   > PlaySound2D(Sound=Get InflictionSound() -> (InflictionSound:Sound Base Object Reference))
   > "SetText (Text)"(self=Get AlertText() -> (AlertText:Common Text Object Reference), InText=Get InflictionText() -> (InflictionText:Text))
   > n3 = SetText(
    self=Get Help_RichText() -> (Help_RichText:Common Rich Text Block Object Reference)
    InText=Get Infliction_Text_Help() -> (Infliction_Text_Help:Text)
   )
   Anim_Load(execute=[@n5.then, @n3.then, @n8.then]) {is=title K2Node_CallFunction}
   > SetVisibility(InVisibility=HitTestInvisible)
   > SetHasActiveContent(bNewActive=true)
   Start Storm Sickness Warning() {is=title K2Node_CustomEvent}
   > Clear and Invalidate Timer by Handle(Handle=Get HideTimerHandle())
   > n4 = Check Show or Hide Widget()
   StopDisplayUI(execute=@n4.False)
   PlaySound2D(execute=@n4.Show, Sound=Get WarningSound() -> (WarningSound:Sound Base Object Reference))
   > "SetText (Text)"(self=Get AlertText(), InText=Get WarningText() -> (WarningText:Text))
   > n5 = SetText(self=Get Help_RichText(), InText=Get Warning_Text_Help() -> (Warning_Text_Help:Text))
   Start Storm Recover Notification() {is=title K2Node_CustomEvent}
   > Clear and Invalidate Timer by Handle(Handle=Get HideTimerHandle())
   > Set bShowingInflicted(bShowingInflicted=false)
   > n6 = Check Show or Hide Widget()
   StopDisplayUI(execute=@n6.False)
   "SetText (Text)"(execute=@n6.Show, self=Get AlertText(), InText=Get Recovery_Text() -> (Recovery_Text:Text))
   > SetText(self=Get Help_RichText(), InText=Get Recovery_Text_Help() -> (Recovery_Text_Help:Text))
   > n7 = Set Timer by Event(Delegate=Create Event(), Time=Get DisplayTime() -> (DisplayTime:double), bLooping=false)
   > n8 = Set HideTimerHandle(HideTimerHandle:Timer Handle Structure=@n7) -> (Output_Get:Timer Handle Structure)
   n9 = Handle Player Controller Target Changed!() -> (
    OutputDelegate
    then
    PlayerController:Fort Player Controller Object Reference
    OldViewTarget:Actor Object Reference
    NewViewTarget:Actor Object Reference
   ) {is=title K2Node_CustomEvent}
   > n10 = Cast To FortPlayerPawnAthena(Object=@n9.NewViewTarget)
   > Check Set Cached Pawn!(execute, self, New Pawn:Fort Player Pawn Athena Object Reference=@n10) -> (then) {is=title K2Node_CallFunction}
   > InStormCheck()
   n11 = Cast To FortLiveBroadcastController(execute=@n10.CastFailed, Object=GetOwningPlayer())
   > n12 = Branch(Condition="Not Equal (Enum)"!(
     A:ESpectatorCameraType Enum=Get CurrentCameraType!(self:Fort Player Controller Spectating Object Reference=@n11) -> (CurrentCameraType:ESpectatorCameraType Enum=ThirdPerson)
     B:ESpectatorCameraType Enum=DroneFree
    ) -> (ReturnValue:bool))
   StopDisplayUI(execute=@n12.else)
   SetBinds() {is=title K2Node_CustomEvent}
   > Set bHasFinishedSetBinds(bHasFinishedSetBinds:bool=false) -> (Output_Get:bool)
   > ResetState()
   > StopDisplayUI()
   n13 = Get CachedPawnToUse()
   > n14 = WaitGameplayEventToActor!(
    execute:Exec
    TargetActor:Actor Object Reference=@n13
    EventTag:Gameplay Tag Structure=(TagName="Event.StormSicknessWarning")
    OnlyTriggerOnce:bool=false
    OnlyMatchExact:bool=true
   ) -> (then:Exec, AsyncTaskProxy:Ability Async Wait Gameplay Event Object Reference, EventReceived:Exec, Payload:Gameplay Event Data Structure) {is=title K2Node_AsyncAction}
   > n15 = WaitGameplayEventToActor(TargetActor=@n13, EventTag=(TagName="Event.stormSickness.ThresholdUpdate"), OnlyTriggerOnce=false, OnlyMatchExact=true)
   > n16 = WaitGameplayTagAddToActor!(
    execute:Exec
    TargetActor:Actor Object Reference=@n13
    Tag:Gameplay Tag Structure=(TagName="Gameplay.InsideSafeZone")
    OnlyTriggerOnce:bool=false
   ) -> (then:Exec, AsyncTaskProxy:Ability Async Wait Gameplay Tag Added Object Reference, Added:Exec) {is=title K2Node_AsyncAction}
   > n17 = WaitGameplayTagRemoveFromActor!(
    execute:Exec
    TargetActor:Actor Object Reference=@n13
    Tag:Gameplay Tag Structure=(TagName="Gameplay.InsideSafeZone")
    OnlyTriggerOnce:bool=false
   ) -> (then:Exec, AsyncTaskProxy:Ability Async Wait Gameplay Tag Removed Object Reference, Removed:Exec) {is=title K2Node_AsyncAction}
   > n18 = WaitGameplayTagAddToActor(TargetActor=@n13, Tag=(TagName="Gameplay.Status.StormSickness"), OnlyTriggerOnce=false)
   > n19 = WaitGameplayTagRemoveFromActor(TargetActor=@n13, Tag=(TagName="Gameplay.Status.StormSickness"), OnlyTriggerOnce=false)
   n20 = Handle Player Died!() -> (OutputDelegate, then, FortPlayerPawnAthena:Fort Player Pawn Athena Object Reference) {is=title K2Node_CustomEvent}
   Bind Event to On Fort Player Died(execute=@n19.then, self:Fort Player Pawn Athena Object Reference=@n13, Delegate=@n20) {is=title K2Node_AddDelegate}
   > Set bHasFinishedSetBinds(bHasFinishedSetBinds=true)
   StopDisplayUI(execute=@n19.Removed)
   Set bIsInflicted(execute=@n18.Added, bIsInflicted:bool=true) -> (Output_Get:bool)
   > Branch(Condition=Get bHasFinishedSetBinds() -> (bHasFinishedSetBinds:bool))
   > Start Storm Sickness Infliction() {is=title K2Node_CallFunction}
   Branch(execute=@n17.Removed, Condition=Get bHasFinishedSetBinds())
   > InStormCheck()
   Branch(execute=@n16.Added, Condition=Get bHasFinishedSetBinds())
   > SafeZoneCheck()
   SetPercent!(
    execute=@n15.EventReceived
    self:Progress Bar Object Reference=Get Sickness_Prog() -> (Sickness_Prog:Progress Bar Object Reference)
    InPercent:float=Break Gameplay Event Data!(GameplayEventData:"Gameplay Event Data Structure (by ref)"=@n15.Payload) -> (
     EventTag:Gameplay Tag Structure
     Instigator:Actor Object Reference
     Target:Actor Object Reference
     OptionalObject:Object Reference
     OptionalObject2:Object Reference
     ContextHandle:Gameplay Effect Context Handle Structure
     InstigatorTags:Gameplay Tag Container Structure
     TargetTags:Gameplay Tag Container Structure
     EventMagnitude:float=~
     TargetData:Gameplay Ability Target Data Handle Structure
     InstancedEventData:Instanced Struct Structure
    ) {is=title K2Node_BreakStruct}.EventMagnitude
   ) -> (then)
   Set bWarned(execute=@n14.EventReceived, bWarned:bool=true) -> (Output_Get:bool)
   > Start Storm Sickness Warning() {is=title K2Node_CallFunction}
   Event Construct()
   > SetVisibility(InVisibility=Collapsed)
   > n21 = Cast To FortPlayerController(Object=GetOwningPlayer())
   > Bind Event to On View Target Changed(self:Fort Player Controller Object Reference=@n21, Delegate=@n9) {is=title K2Node_AddDelegate}
   n22 = GetOwningPlayerPawn!(self:User Widget Object Reference) -> (ReturnValue:Pawn Object Reference)
   > n23 = Is Valid(InputObject=@n22)
   PrintString!(
    execute=@n23.Is Not Valid
    InString="AthenaStormSicknessWarning: No Owning Player Pawn!!"
    bPrintToScreen
    bPrintToLog
    TextColor
    Duration
    Key
   ) -> (then)
   n24 = Cast To FortPlayerPawnAthena(execute=@n23.Is Valid, Object=@n22)
   > Set CachedPawnToUse(CachedPawnToUse=@n24)
   > SetBinds() {is=title K2Node_CallFunction}
   PrintString(execute=@n21.CastFailed, InString="AthenaStormSicknessWarning: FortPlayerController cast failed!")
   StopDisplayUI(execute=@n20.then)
   Reroute Node(InputPin:Wildcard) -> (OutputPin:Wildcard)
  ]}
 GetCacheContextOverride {nodes [
   GetCacheContextOverride() {is=title K2Node_FunctionEntry}
   > Return Node!(execute, ReturnValue:Object Reference)
  ]}
 InStormCheck {nodes [
   InStormCheck() {is=title K2Node_FunctionEntry}
   > n1 = Branch(Condition=HasMatchingGameplayTag(self=Get CachedPawnToUse(), TagToCheck=(TagName="Gameplay.Status.StormSickness")))
   > n2 = Branch(Condition=Get bShowingInflicted() -> (bShowingInflicted:bool))
   Start Storm Sickness Infliction(execute=@n2.else)
   n3 = Branch(execute=@n1.else, Condition=Get bWarned() -> (bWarned:bool))
   > Start Storm Sickness Warning()
   StopDisplayUI(execute=@n3.else)
  ]}
 ResetState {nodes [
   ResetState() {is=title K2Node_FunctionEntry}
   > Set bWarned(bWarned=false)
   > Set bIsInflicted(bIsInflicted=false)
   > Set bShowingInflicted(bShowingInflicted=false)
  ]}
 SafeZoneCheck {nodes [
   SafeZoneCheck() {is=title K2Node_FunctionEntry}
   > n1 = Branch(Condition=AND Boolean(A=Get bIsInflicted() -> (bIsInflicted:bool), B=Get bShowingInflicted()))
   > Start Storm Recover Notification() {is=title K2Node_CallFunction}
   StopDisplayUI(execute=@n1.else)
  ]}
 StartStormSurgeDisplay {nodes [StartStormSurgeDisplay() {is=title K2Node_FunctionEntry}]}
 StopStormSurgeDisplay {nodes [StopStormSurgeDisplay() {is=title K2Node_FunctionEntry}]}
]
variables [
 bHasFinishedSetBinds {type=boolean}
 bIsInflicted {type=boolean}
 bShowingInflicted {type=boolean}
 bStormSicknessEnabled {type=ScalableFloat, default {registryType=Fort.DataRegistry, value=1, curve {curveTable=null, rowName=StormSickness.Enabled}}}
 bWarned {type=boolean}
 CachedPawnToUse {type=FortPlayerPawnAthena}
 CachedPlayerController {type=FortPlayerController, default=null}
 DisplayTime {type=number, default=10}
 HideTimerHandle {type=TimerHandle}
 Infliction_Text_Help {type=string, default="Exit the Storm to <em>remove</> storm sickness"}
 InflictionSound {type=SoundBase, default=/Game/Athena/Sounds/LTMs/StormCap/Sting_StormSickness_DamageInflicted_Cue}
 InflictionText {type=string, default="THE STORM RAGES!"}
 Recovery_Text {type=string, default="YOU'VE RECOVERED!"}
 Recovery_Text_Help {type=string, default="You will be <em>afflicted</> if you enter the storm again"}
 StartStormSurgeDisplay {}
 StopStormSurgeDisplay {}
 Warning_Text_Help {type=string, default="Exit the Storm to <em>avoid</> storm sickness"}
 WarningSound {type=SoundBase, default=/Game/Athena/Sounds/LTMs/StormCap/Sting_StormSickness_Warning_Cue}
 WarningText {type=string, default="STORM SICKNESS WARNING!"}
]
widgets {
 info {
  inheritedWidgetCount=1
  parentClass=/Script/StormSicknessUI.StormSicknessWarningWidgetBase
  rootWidgetClass=&Overlay
  widgetCount=12
 }
 widgets [
  Overlay_1 {widgetClassPath=&Overlay, widget=Overlay_1}
  LightningBack {
   widgetClassPath=&Image
   bIsVariable=true
   parent=Overlay_1
   widget=LightningBack
   properties {
    renderOpacity=0.5
    brush {
     bIsDynamicallyLoaded=false
     drawAs=Image
     imageType=NoImage
     mirroring=NoMirror
     resourceName=None
     resourceObject=/Game/Athena/HUD/Materials/M_UI_LightningFX_Inst
     tiling=NoTile
     imageSize=<398.6756 64>
     margin=<ltrb 0 0 0 0>
     outlineSettings {
      bUseBrushTransparency=false
      roundingType=HalfHeightRadius
      width=0
      color {colorUseRule=UseColor_Specified, specifiedColor=<rgba 0 0 0 0>}
      cornerRadii=<0 0 0 0>
     }
     tintColor {colorUseRule=UseColor_Specified, specifiedColor=<rgba 1 1 1 1>}
     uVRegion {bIsValid=false, max=<0 0>, min=<0 0>}
    }
    renderTransform {angle=0, scale=<1 1.25>, shear=<0 0>, translation=<-48 0>}
   }
   slot {class=OverlaySlot, properties {horizontalAlignment=HAlign_Fill, verticalAlignment=VAlign_Center, padding=<ltrb 0 0 48 0>}}
  }
  HorizontalBox_0 {widgetClassPath=&HorizontalBox, parent=Overlay_1, widget=HorizontalBox_0, slot {class=OverlaySlot}}
  Overlay_0 {
   widgetClassPath=&Overlay
   parent=HorizontalBox_0
   widget=Overlay_0
   slot {class=HorizontalBoxSlot, properties {horizontalAlignment=HAlign_Center, verticalAlignment=VAlign_Center}}
  }
  Glow {
   widgetClassPath=&Image
   bIsVariable=true
   parent=Overlay_0
   widget=Glow
   properties {
    renderOpacity=0.5
    brush {
     bIsDynamicallyLoaded=false
     drawAs=Image
     imageType=NoImage
     mirroring=NoMirror
     resourceName=None
     resourceObject=/Game/Athena/HUD/LTM/Wax/Materials/M_UI_TextGlow_Add_Gradient_StormSurge
     tiling=NoTile
     imageSize=<32 32>
     margin=<ltrb 0 0 0 0>
     outlineSettings {
      bUseBrushTransparency=false
      roundingType=HalfHeightRadius
      width=0
      color {colorUseRule=UseColor_Specified, specifiedColor=<rgba 0 0 0 0>}
      cornerRadii=<0 0 0 0>
     }
     tintColor {colorUseRule=UseColor_Specified, specifiedColor=<rgba 1 1 1 1>}
     uVRegion {bIsValid=false, max=<0 0>, min=<0 0>}
    }
    renderTransform {angle=0, scale=<1.375 1.125>, shear=<0 0>, translation=<0 8>}
   }
   slot {class=OverlaySlot, properties {horizontalAlignment=HAlign_Fill, verticalAlignment=VAlign_Fill}}
  }
  StormIcon {
   widgetClassPath=/Script/FortniteUI.FortMobileImage
   bIsVariable=true
   parent=Overlay_0
   widget=StormIcon
   properties {
    mobileSizeMultiplier=1.4
    brush {
     bIsDynamicallyLoaded=false
     drawAs=Image
     imageType=NoImage
     mirroring=NoMirror
     resourceName=None
     resourceObject=&/T_UI_StormSickness_128x
     tiling=NoTile
     imageSize=<64 64>
     margin=<ltrb 0 0 0 0>
     outlineSettings {
      bUseBrushTransparency=false
      roundingType=HalfHeightRadius
      width=0
      color {colorUseRule=UseColor_Specified, specifiedColor=<rgba 0 0 0 0>}
      cornerRadii=<0 0 0 0>
     }
     tintColor {colorUseRule=UseColor_Specified, specifiedColor=<rgba 1 1 1 1>}
     uVRegion {bIsValid=false, max=<0 0>, min=<0 0>}
    }
   }
   slot {class=OverlaySlot, properties {horizontalAlignment=HAlign_Center, verticalAlignment=VAlign_Bottom}}
  }
  IconLightning {
   widgetClassPath=&Image
   bIsVariable=true
   parent=Overlay_0
   widget=IconLightning
   properties {
    renderOpacity=0.25
    brush {
     bIsDynamicallyLoaded=false
     drawAs=Image
     imageType=NoImage
     mirroring=NoMirror
     resourceName=None
     resourceObject=/Game/Athena/HUD/XP/Materials/M_UI_PwrUp_Lightning
     tiling=NoTile
     imageSize=<32 32>
     margin=<ltrb 0 0 0 0>
     outlineSettings {
      bUseBrushTransparency=false
      roundingType=HalfHeightRadius
      width=0
      color {colorUseRule=UseColor_Specified, specifiedColor=<rgba 0 0 0 0>}
      cornerRadii=<0 0 0 0>
     }
     tintColor {colorUseRule=UseColor_Specified, specifiedColor=<rgba 1 1 1 1>}
     uVRegion {bIsValid=false, max=<0 0>, min=<0 0>}
    }
    renderTransform {angle=0, scale=<1.5 1.5>, shear=<0 0>, translation=<0 0>}
   }
   slot {class=OverlaySlot, properties {horizontalAlignment=HAlign_Fill, verticalAlignment=VAlign_Fill}}
  }
  VerticalBox_2 {
   widgetClassPath=&VerticalBox
   parent=HorizontalBox_0
   widget=VerticalBox_2
   slot {class=HorizontalBoxSlot, properties {verticalAlignment=VAlign_Center, padding=<ltrb 0 2 0 0>}}
  }
  AlertText {
   widgetClassPath=&CommonTextBlock
   bIsVariable=true
   parent=VerticalBox_2
   widget=AlertText
   properties {
    mobileFontSizeMultiplier=1.25
    text="[PH] STORM SICKNESS WARNING! "
    font {
     bForceMonospaced=false
     bMaterialIsStencil=false
     fontMaterial=null
     fontObject=&Burbank
     letterSpacing=0
     monospacedWidth=1
     size=20
     skewAmount=0
     typefaceFontName=BurbankRegular-Black
     outlineSettings {
      bApplyOutlineToDropShadows=false
      bMiteredCorners=false
      bSeparateFillAlpha=false
      outlineMaterial=null
      outlineSize=3
      outlineColor=<rgba 0.0232 0 0.0307 0.125>
     }
    }
    shadowOffset=<0 0>
   }
   slot {
    class=VerticalBoxSlot
    properties {horizontalAlignment=HAlign_Left, verticalAlignment=VAlign_Center, padding=<ltrb -12 -8 0 0>, size {sizeRule=Fill, value=1}}
   }
  }
  Help_RichText {
   widgetClassPath=&CommonRichTextBlock
   bIsVariable=true
   parent=VerticalBox_2
   widget=Help_RichText
   properties {
    mobileTextBlockScale=1.25
    text="[PH] Exit the Storm to <em>avoid</> higher damage!"
    textStyleSet=&/StormSickness_RichTextStyleSet
    textTransformPolicy=ToUpper
    wrapTextAt=650
    defaultTextStyleOverride {
     overflowPolicy=Clip
     transformPolicy=None
     colorAndOpacity {colorUseRule=UseColor_Specified, specifiedColor=<rgba 1 1 1 1>}
     font {
      bForceMonospaced=false
      bMaterialIsStencil=false
      fontMaterial=null
      fontObject=&NotoSans
      letterSpacing=0
      monospacedWidth=1
      size=24
      skewAmount=0
      typefaceFontName=NotoSans-Regular
      outlineSettings {
       bApplyOutlineToDropShadows=false
       bMiteredCorners=false
       bSeparateFillAlpha=false
       outlineMaterial=null
       outlineSize=0
       outlineColor=<rgba 0 0 0 1>
      }
     }
     highlightColor {colorUseRule=UseColor_Specified, specifiedColor=<rgba 0 0 0 1>}
     highlightShape {
      bIsDynamicallyLoaded=false
      drawAs=Image
      imageType=NoImage
      mirroring=NoMirror
      resourceName=None
      resourceObject=null
      tiling=NoTile
      imageSize=<32 32>
      margin=<ltrb 0 0 0 0>
      outlineSettings {
       bUseBrushTransparency=false
       roundingType=HalfHeightRadius
       width=0
       color {colorUseRule=UseColor_Specified, specifiedColor=<rgba 0 0 0 0>}
       cornerRadii=<0 0 0 0>
      }
      tintColor {colorUseRule=UseColor_Specified, specifiedColor=<rgba 1 1 1 1>}
      uVRegion {bIsValid=false, max=<0 0>, min=<0 0>}
     }
     selectedBackgroundColor {colorUseRule=UseColor_Foreground, specifiedColor=<rgba 1 0 1 1>}
     shadowColorAndOpacity=<rgba 0 0 0 1>
     shadowOffset=<0 0>
     strikeBrush {
      bIsDynamicallyLoaded=false
      drawAs=Image
      imageType=NoImage
      mirroring=NoMirror
      resourceName=None
      resourceObject=null
      tiling=NoTile
      imageSize=<32 32>
      margin=<ltrb 0 0 0 0>
      outlineSettings {
       bUseBrushTransparency=false
       roundingType=HalfHeightRadius
       width=0
       color {colorUseRule=UseColor_Specified, specifiedColor=<rgba 0 0 0 0>}
       cornerRadii=<0 0 0 0>
      }
      tintColor {colorUseRule=UseColor_Specified, specifiedColor=<rgba 1 1 1 1>}
      uVRegion {bIsValid=false, max=<0 0>, min=<0 0>}
     }
     underlineBrush {
      bIsDynamicallyLoaded=false
      drawAs=Image
      imageType=NoImage
      mirroring=NoMirror
      resourceName=None
      resourceObject=null
      tiling=NoTile
      imageSize=<32 32>
      margin=<ltrb 0 0 0 0>
      outlineSettings {
       bUseBrushTransparency=false
       roundingType=HalfHeightRadius
       width=0
       color {colorUseRule=UseColor_Specified, specifiedColor=<rgba 0 0 0 0>}
       cornerRadii=<0 0 0 0>
      }
      tintColor {colorUseRule=UseColor_Specified, specifiedColor=<rgba 1 1 1 1>}
      uVRegion {bIsValid=false, max=<0 0>, min=<0 0>}
     }
    }
   }
   slot {class=VerticalBoxSlot, properties {verticalAlignment=VAlign_Top, padding=<ltrb 10 -8 0 0>}}
  }
  Sickness_Prog {
   widgetClassPath=/Script/UMG.ProgressBar
   bIsVariable=true
   parent=VerticalBox_2
   widget=Sickness_Prog
   properties {
    visibility=Collapsed
    fillColorAndOpacity=<rgba 0.3854 0 0.3821 1>
    widgetStyle {
     enableFillAnimation=false
     backgroundImage {
      bIsDynamicallyLoaded=false
      drawAs=Box
      imageType=FullColor
      mirroring=NoMirror
      resourceName=../../../Engine/Content/Slate/Common/ProgressBar_Background.png
      resourceObject=null
      tiling=NoTile
      imageSize=<0 0>
      margin=<ltrb 0.4167 0.4167 0.4167 0.4167>
      outlineSettings {
       bUseBrushTransparency=false
       roundingType=HalfHeightRadius
       width=0
       color {colorUseRule=UseColor_Specified, specifiedColor=<rgba 0 0 0 0>}
       cornerRadii=<0 0 0 0>
      }
      tintColor {colorUseRule=UseColor_Specified, specifiedColor=<rgba 0 0 0 0.33>}
      uVRegion {bIsValid=false, max=<0 0>, min=<0 0>}
     }
     fillImage {
      bIsDynamicallyLoaded=false
      drawAs=Box
      imageType=FullColor
      mirroring=NoMirror
      resourceName=../../../Engine/Content/Slate/Common/ProgressBar_Fill.png
      resourceObject=null
      tiling=NoTile
      imageSize=<0 0>
      margin=<ltrb 0.4167 0.4167 0.4167 0.4167>
      outlineSettings {
       bUseBrushTransparency=false
       roundingType=HalfHeightRadius
       width=0
       color {colorUseRule=UseColor_Specified, specifiedColor=<rgba 0 0 0 0>}
       cornerRadii=<0 0 0 0>
      }
      tintColor {colorUseRule=UseColor_Specified, specifiedColor=<rgba 1 1 1 1>}
      uVRegion {bIsValid=false, max=<0 0>, min=<0 0>}
     }
     marqueeImage {
      bIsDynamicallyLoaded=false
      drawAs=Image
      imageType=FullColor
      mirroring=NoMirror
      resourceName=../../../Engine/Content/Slate/Common/ProgressBar_Marquee.png
      resourceObject=null
      tiling=Horizontal
      imageSize=<20 12>
      margin=<ltrb 0 0 0 0>
      outlineSettings {
       bUseBrushTransparency=false
       roundingType=HalfHeightRadius
       width=0
       color {colorUseRule=UseColor_Specified, specifiedColor=<rgba 0 0 0 0>}
       cornerRadii=<0 0 0 0>
      }
      tintColor {colorUseRule=UseColor_Specified, specifiedColor=<rgba 1 1 1 1>}
      uVRegion {bIsValid=false, max=<0 0>, min=<0 0>}
     }
    }
   }
   slot {class=VerticalBoxSlot, properties {padding=<ltrb 0 5 0 0>}}
  }
  TutorialHighlight {widgetClassPath=&TutorialHighlight}
 ]
}
