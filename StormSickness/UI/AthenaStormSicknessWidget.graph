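# Storm Sickness HUD element that shows a static alert telling the player they are afflicted, with storm icon, glow and lightning backdrop.
# On Construct it plays the DMGLoad intro (icon pop, text slide and fade) and becomes hit-test invisible.
# PreConstruct realigns icon and text when isBacchus is set.
# It scopes data registry lookups to the owning player; infection, safe-zone and sound variables are unused in its graphs.
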WidgetBlueprint /StormSickness/UI/AthenaStormSicknessWidget
animations [
 DMGLoad {
  binding_count=5
  display_label=""
  length_seconds=0.65
  object=&self:DMGLoad
  sequence {
   bindings [
    AlertText {
     id=#1
     possessed_class=CommonTextBlock
     tracks [
      {
       type=MovieSceneFloatTrack
       display_name=Render Opacity
       sections [{type=MovieSceneFloatSection, channels [None {type=MovieSceneScriptingFloatChannel, keys [[-3,0],[3,1]]}]}]
      }
      {
       type=MovieScene2DTransformTrack
       sections [{type=MovieScene2DTransformSection, channels [Translation.X {type=MovieSceneScriptingFloatChannel, keys [[-3,80],[6,0]]}]}]
      }
     ]
    }
    HelpText {
     id=#2
     possessed_class=CommonTextBlock
     tracks [
      {
       type=MovieScene2DTransformTrack
       sections [{type=MovieScene2DTransformSection, channels [Translation.X {type=MovieSceneScriptingFloatChannel, keys [[0,80],[9,0]]}]}]
      }
      {
       type=MovieSceneFloatTrack
       display_name=Render Opacity
       sections [{type=MovieSceneFloatSection, channels [None {type=MovieSceneScriptingFloatChannel, keys [[0,0],[6,1]]}]}]
      }
     ]
    }
    LightningBack {
     id=#3
     possessed_class=Image
     tracks [
      {
       type=MovieSceneFloatTrack
       display_name=Render Opacity
       sections [{type=MovieSceneFloatSection, channels [None {type=MovieSceneScriptingFloatChannel, keys [[-12,0],[-11,1],[0,0.5]]}]}]
      }
      {
       type=MovieScene2DTransformTrack
       sections [{
         type=MovieScene2DTransformSection
         channels [
          Translation.X {type=MovieSceneScriptingFloatChannel, const [48,1]}
          Translation.Y {type=MovieSceneScriptingFloatChannel}
          Angle {type=MovieSceneScriptingFloatChannel}
          Scale.X {type=MovieSceneScriptingFloatChannel, keys [[-12,5],[-6,1.75]]}
          Scale.Y {type=MovieSceneScriptingFloatChannel, const [0.9,1]}
         ]
        }]
      }
     ]
    }
    SicknessIcon_SizeBox {
     id=#4
     possessed_class=SizeBox
     tracks [{
       type=MovieScene2DTransformTrack
       sections [{
         type=MovieScene2DTransformSection
         channels [
          Scale.X {type=MovieSceneScriptingFloatChannel, keys [[-3,0],[3,1.125],[9,1]]}
          Scale.Y {type=MovieSceneScriptingFloatChannel, keys [[-3,0],[3,1.125],[9,1]]}
         ]
        }]
      }]
    }
    Glow_1 {
     id=#5
     possessed_class=Image
     tracks [{
       type=MovieSceneFloatTrack
       display_name=Render Opacity
       sections [{type=MovieSceneFloatSection, channels [None {type=MovieSceneScriptingFloatChannel, keys [[-30,0],[-12,0],[0,0.0625]]}]}]
      }]
    }
   ]
   display_rate {numerator=60}
  }
 }
 Intro {binding_count=0, length_seconds=5, object=&self:Intro, sequence {bindings []}}
]
graphs [
 Anim_Load {nodes [
   Anim_Load() {is=title K2Node_FunctionEntry}
   > PlayAnimation(InAnimation=Get DMGLoad() -> (DMGLoad:Widget Animation Object Reference))
  ]}
 BuildDataRegistryResolverScope {nodes [
   BuildDataRegistryResolverScope() {is=title K2Node_FunctionEntry}
   n1 = GetOwningPlayer()
   > Return Node!(
    execute
    InOutResolverScopes:Array of Names=Make Array("[0]":Name=GetSpatialGameplayInterfaceFNameForActor!(Actor:Actor Object Reference=@n1) -> (ReturnValue:Name=None)) -> (Array:Array of Names)
    InOutPriority:int=1
    ReturnValue:bool=true
   )
  ]}
 EventGraph {nodes [
   Event PreConstruct()
   > Branch(Condition=Get isBacchus() -> (isBacchus:bool))
   > SetHorizontalAlignment(
    self=SlotAsOverlaySlot(Widget=Get SicknessText_SizeBox() -> (SicknessText_SizeBox:Size Box Object Reference))
    InHorizontalAlignment=HAlign_Left
   )
   n1 = Get SicknessIcon_SizeBox() -> (SicknessIcon_SizeBox:Size Box Object Reference)
   > SetHorizontalAlignment(self=SlotAsOverlaySlot(Widget=@n1), InHorizontalAlignment=HAlign_Right)
   > SetRenderTranslation(self=@n1, Translation=(X=10,Y=0))
   Event Construct()
   > Anim_Load() {is=title K2Node_CallFunction}
   > SetVisibility(InVisibility=HitTestInvisible)
  ]}
 GetCacheContextOverride {nodes [
   GetCacheContextOverride() {is=title K2Node_FunctionEntry}
   > Return Node!(execute, ReturnValue:Object Reference)
  ]}
]
variables [
 bAboveIsOpen {type=boolean}
 bBelowIsOpen {type=boolean}
 bIsDisplaying {type=boolean}
 bIsInfected {type=boolean}
 bIsInSafeZone {type=boolean}
 bStormSicknessEnabled {
  type=ScalableFloat
  default {registryType=None, value=1, curve {curveTable=/StormSickness/DataTables/StormSicknessGameData, rowName=StormSickness.Enabled}}
 }
 CachedPawnToUse {type=FortPlayerPawnAthena}
 isBacchus {type=boolean}
 SoundAboveOpen {type=SoundBase, default=/Game/Athena/Sounds/LTMs/StormCap/StormCap_UIAlert_Above_Cue}
 SoundBelowOpen {type=SoundBase, default=/Game/Athena/Sounds/LTMs/StormCap/StormCap_UIAlert_Below_Cue}
]
widgets {
 info {inheritedWidgetCount=1, parentClass=&FortHUDElementWidget, rootWidgetClass=&SizeBox, widgetCount=15}
 widgets [
  RootSizeBox {widgetClassPath=&SizeBox, bIsVariable=true, widget=RootSizeBox, properties {widthOverride=290}}
  Overlay_1 {
   widgetClassPath=&Overlay
   parent=RootSizeBox
   widget=Overlay_1
   slot {class=SizeBoxSlot, properties {verticalAlignment=VAlign_Center}}
  }
  LightningBack {
   widgetClassPath=&Image
   bIsVariable=true
   parent=Overlay_1
   widget=LightningBack
   properties {
    renderOpacity=0.5
    brush {
     bIsDynamicallyLoaded=false
     drawAs=Image
     imageType=NoImage
     mirroring=NoMirror
     resourceName=None
     resourceObject=/Game/Athena/HUD/Materials/M_UI_LightningFX_Inst
     tiling=NoTile
     imageSize=<237.5153 64>
     margin=<ltrb 0 0 0 0>
     outlineSettings {
      bUseBrushTransparency=false
      roundingType=HalfHeightRadius
      width=0
      color {colorUseRule=UseColor_Specified, specifiedColor=<rgba 0 0 0 0>}
      cornerRadii=<0 0 0 0>
     }
     tintColor {colorUseRule=UseColor_Specified, specifiedColor=<rgba 1 1 1 1>}
     uVRegion {bIsValid=false, max=<0 0>, min=<0 0>}
    }
    renderTransform {angle=0, scale=<1.75 0.9>, shear=<0 0>, translation=<48 0>}
   }
   slot {class=OverlaySlot, properties {horizontalAlignment=HAlign_Fill, verticalAlignment=VAlign_Center}}
  }
  Glow_1 {
   widgetClassPath=&Image
   bIsVariable=true
   parent=Overlay_1
   widget=Glow_1
   properties {
    renderOpacity=0.0625
    brush {
     bIsDynamicallyLoaded=false
     drawAs=Image
     imageType=NoImage
     mirroring=NoMirror
     resourceName=None
     resourceObject=/Game/Athena/HUD/LTM/Wax/Materials/M_UI_TextGlow_Add_Gradient_StormSurge
     tiling=NoTile
     imageSize=<32 32>
     margin=<ltrb 0 0 0 0>
     outlineSettings {
      bUseBrushTransparency=false
      roundingType=HalfHeightRadius
      width=0
      color {colorUseRule=UseColor_Specified, specifiedColor=<rgba 0 0 0 0>}
      cornerRadii=<0 0 0 0>
     }
     tintColor {colorUseRule=UseColor_Specified, specifiedColor=<rgba 1 1 1 1>}
     uVRegion {bIsValid=false, max=<0 0>, min=<0 0>}
    }
    renderTransform {angle=0, scale=<3.333 2>, shear=<0 0>, translation=<0 8>}
    renderTransformPivot=<0.325 0.5>
   }
   slot {class=OverlaySlot, properties {horizontalAlignment=HAlign_Fill, verticalAlignment=VAlign_Fill}}
  }
  HorizontalBox_0 {
   widgetClassPath=&HorizontalBox
   parent=Overlay_1
   widget=HorizontalBox_0
   slot {class=OverlaySlot, properties {horizontalAlignment=HAlign_Fill, verticalAlignment=VAlign_Center}}
  }
  SicknessIcon_SizeBox {
   widgetClassPath=&SizeBox
   bIsVariable=true
   parent=HorizontalBox_0
   widget=SicknessIcon_SizeBox
   slot {class=HorizontalBoxSlot, properties {verticalAlignment=VAlign_Center}}
  }
  Overlay_0 {
   widgetClassPath=&Overlay
   parent=SicknessIcon_SizeBox
   widget=Overlay_0
   slot {class=SizeBoxSlot, properties {horizontalAlignment=HAlign_Center, verticalAlignment=VAlign_Center}}
  }
  Glow {
   widgetClassPath=&Image
   bIsVariable=true
   parent=Overlay_0
   widget=Glow
   properties {
    renderOpacity=0.5
    brush {
     bIsDynamicallyLoaded=false
     drawAs=Image
     imageType=NoImage
     mirroring=NoMirror
     resourceName=None
     resourceObject=&M_UI_TextGlow_Add_Gradient_StormSurge
     tiling=NoTile
     imageSize=<32 32>
     margin=<ltrb 0 0 0 0>
     outlineSettings {
      bUseBrushTransparency=false
      roundingType=HalfHeightRadius
      width=0
      color {colorUseRule=UseColor_Specified, specifiedColor=<rgba 0 0 0 0>}
      cornerRadii=<0 0 0 0>
     }
     tintColor {colorUseRule=UseColor_Specified, specifiedColor=<rgba 1 1 1 1>}
     uVRegion {bIsValid=false, max=<0 0>, min=<0 0>}
    }
    renderTransform {angle=0, scale=<1.375 1.125>, shear=<0 0>, translation=<0 8>}
   }
   slot {class=OverlaySlot, properties {horizontalAlignment=HAlign_Fill, verticalAlignment=VAlign_Fill}}
  }
  StormIcon {
   widgetClassPath=/Script/FortniteUI.FortMobileImage
   bIsVariable=true
   parent=Overlay_0
   widget=StormIcon
   properties {
    mobileSizeMultiplier=1.4
    brush {
     bIsDynamicallyLoaded=false
     drawAs=Image
     imageType=NoImage
     mirroring=NoMirror
     resourceName=None
     resourceObject=&/T_UI_StormSickness_128x
     tiling=NoTile
     imageSize=<64 64>
     margin=<ltrb 0 0 0 0>
     outlineSettings {
      bUseBrushTransparency=false
      roundingType=HalfHeightRadius
      width=0
      color {colorUseRule=UseColor_Specified, specifiedColor=<rgba 0 0 0 0>}
      cornerRadii=<0 0 0 0>
     }
     tintColor {colorUseRule=UseColor_Specified, specifiedColor=<rgba 1 1 1 1>}
     uVRegion {bIsValid=false, max=<0 0>, min=<0 0>}
    }
   }
   slot {class=OverlaySlot, properties {horizontalAlignment=HAlign_Center, verticalAlignment=VAlign_Bottom}}
  }
  IconLightning {
   widgetClassPath=&Image
   bIsVariable=true
   parent=Overlay_0
   widget=IconLightning
   properties {
    renderOpacity=0.25
    visibility=Collapsed
    brush {
     bIsDynamicallyLoaded=false
     drawAs=Image
     imageType=NoImage
     mirroring=NoMirror
     resourceName=None
     resourceObject=/Game/Athena/HUD/XP/Materials/M_UI_PwrUp_Lightning
     tiling=NoTile
     imageSize=<32 32>
     margin=<ltrb 0 0 0 0>
     outlineSettings {
      bUseBrushTransparency=false
      roundingType=HalfHeightRadius
      width=0
      color {colorUseRule=UseColor_Specified, specifiedColor=<rgba 0 0 0 0>}
      cornerRadii=<0 0 0 0>
     }
     tintColor {colorUseRule=UseColor_Specified, specifiedColor=<rgba 1 1 1 1>}
     uVRegion {bIsValid=false, max=<0 0>, min=<0 0>}
    }
    renderTransform {angle=0, scale=<1.5 1.5>, shear=<0 0>, translation=<0 0>}
   }
   slot {class=OverlaySlot, properties {horizontalAlignment=HAlign_Fill, verticalAlignment=VAlign_Fill}}
  }
  SicknessText_SizeBox {
   widgetClassPath=&SizeBox
   bIsVariable=true
   parent=HorizontalBox_0
   widget=SicknessText_SizeBox
   properties {widthOverride=236}
   slot {class=HorizontalBoxSlot, properties {verticalAlignment=VAlign_Center}}
  }
  VerticalBox_0 {
   widgetClassPath=&VerticalBox
   parent=SicknessText_SizeBox
   widget=VerticalBox_0
   slot {class=SizeBoxSlot, properties {verticalAlignment=VAlign_Center}}
  }
  AlertText {
   widgetClassPath=&CommonTextBlock
   bIsVariable=true
   parent=VerticalBox_0
   widget=AlertText
   properties {
    justification=Right
    mobileFontSizeMultiplier=1.5
    text=STORM SICKNESS
    colorAndOpacity {colorUseRule=UseColor_Specified, specifiedColor=<rgba 0.9375 0.625 1 1>}
    font {
     bForceMonospaced=false
     bMaterialIsStencil=false
     fontMaterial=/Game/UI/InGame/HUD/Textures/M_StormAlertText
     fontObject=&BurbankSmall
     letterSpacing=0
     monospacedWidth=1
     size=13
     skewAmount=0
     typefaceFontName=Black
     outlineSettings {
      bApplyOutlineToDropShadows=false
      bMiteredCorners=false
      bSeparateFillAlpha=false
      outlineMaterial=null
      outlineSize=2
      outlineColor=<rgba 0.0232 0 0.0307 0.125>
     }
    }
    shadowColorAndOpacity=<rgba 0.0232 0 0.0307 0.8>
    shadowOffset=<0 0>
   }
   slot {class=VerticalBoxSlot, properties {horizontalAlignment=HAlign_Left, padding=<ltrb -8 0 0 -8>}}
  }
  HelpText {
   widgetClassPath=&CommonTextBlock
   bIsVariable=true
   parent=VerticalBox_0
   widget=HelpText
   properties {
    justification=Right
    mobileFontSizeMultiplier=1.3
    text=You're Afflicted
    textTransformPolicy=ToUpper
    font {
     bForceMonospaced=false
     bMaterialIsStencil=false
     fontMaterial=null
     fontObject=&Burbank
     letterSpacing=0
     monospacedWidth=1
     size=18
     skewAmount=0
     typefaceFontName=BurbankRegular-Black
     outlineSettings {
      bApplyOutlineToDropShadows=false
      bMiteredCorners=false
      bSeparateFillAlpha=false
      outlineMaterial=null
      outlineSize=3
      outlineColor=<rgba 0.0232 0 0.0307 0.125>
     }
    }
    shadowColorAndOpacity=<rgba 0.0232 0 0.0307 0.8>
    shadowOffset=<0 0>
   }
   slot {class=VerticalBoxSlot, properties {horizontalAlignment=HAlign_Right, verticalAlignment=VAlign_Bottom}}
  }
  TutorialHighlight {widgetClassPath=&TutorialHighlight}
 ]
}
