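; Dynamic UI director that shows and hides the Storm Sickness scene for the viewed player pawn.
; When enabled, non-touch and in zone, it caches the pawn and re-caches it on view target changes.
; Tag wait tasks show the scene for storm sickness outside the safe zone, gated by a broadcast/camera check.
; A SceneActive flag guards add/remove, but death and non-pawn view targets remove the scene without clearing it.
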(event EventBeginPlay
  ;[0 1 0.25] Initialize the UI Director
  (if (ScalableFloat|GetValueasBoolwithContext (Variables|Default|GetbStormSicknessEnabled))
    (if (TouchPlatform|IsUsingTouch (DynamicUIDirector|GetOwningLocalPlayerController))
      _
      (elif (GlobalUIContext|IsinZone (PlayerController|LocalPlayerSubsystems|GetFortGlobalUIContext (DynamicUIDirector|GetOwningLocalPlayerController)))
        (bind _asfort_player_controller (Utilities|Casting|CastToFortPlayerController (DynamicUIDirector|GetOwningLocalPlayerController)))
        (bind _asfort_player_pawn_athena (Utilities|Casting|CastToFortPlayerPawnAthena (Game|Player|GetPlayerPawn _asfort_player_controller))
          (:then
            (Variables|Default|SetCachedPlayerPawn _asfort_player_pawn_athena)
            (CallFunction|ListentoPawnTags))
          (:CastFailed
            _))
        (EventDispatchers|BindEventtoOnViewTargetChanged _asfort_player_controller (delegate Custom|ViewTargetChanged))))))

(event Custom|ViewTargetChanged
  (params (PlayerController :type Fort%20Player%20Controller%20Object%20Reference) (OldViewTarget :type Actor%20Object%20Reference) (NewViewTarget :type Actor%20Object%20Reference))
  ;[1 1 1] When the Player Controller changes its target, reset the UI Director
  (bind _asfort_player_pawn_athena (Utilities|Casting|CastToFortPlayerPawnAthena NewViewTarget)
    (:then
      (Variables|Default|SetCachedPlayerPawn _asfort_player_pawn_athena)
      (CallFunction|ListentoPawnTags))
    (:CastFailed
      (Utilities|CheckBroadcastandCamera
        (:Hide
          (DynamicUIDirector|RemoveScene "/StormSickness/UI/StormSickness_Scene.StormSickness_Scene"))))))

(event Custom|ListenToPawnTags
  (bind _cachedplayerpawn (Variables|Default|GetCachedPlayerPawn))
  (bind _cachedplayerpawn_2 (Variables|Default|GetCachedPlayerPawn))
  ;[1 1 1] Inside safe zone
  (Ability|Tasks|WaitGameplayTagAddToActor (Variables|Default|GetCachedPlayerPawn) "(TagName=\"Gameplay.InsideSafeZone\")"
    (:then
      (Ability|Tasks|WaitGameplayTagRemoveFromActor _cachedplayerpawn "(TagName=\"Gameplay.InsideSafeZone\")"
        (:then
          ;[1 1 1] Storm Sickness
          (Ability|Tasks|WaitGameplayTagAddToActor _cachedplayerpawn_2 "(TagName=\"Gameplay.Status.StormSickness\")"
            (:then
              (Ability|Tasks|WaitGameplayTagRemoveFromActor (Variables|Default|GetCachedPlayerPawn) "(TagName=\"Gameplay.Status.StormSickness\")"
                (:then
                  ;[1 1 1] Death
                  (EventDispatchers|BindEventtoOnFortPlayerDied (Variables|Default|GetCachedPlayerPawn) (delegate Custom|OnFortPlayerDied)))
                (:Removed
                  (CallFunction|RemoveSceneifActive))))
            (:Added
              (if (not (GameplayTags|HasMatchingGameplayTag _cachedplayerpawn_2 "(TagName=\"Gameplay.InsideSafeZone\")"))
                (Utilities|CheckBroadcastandCamera
                  (:Show
                    (CallFunction|AddSceneifUnactive)))))))
        (:Removed
          (if (GameplayTags|HasMatchingGameplayTag _cachedplayerpawn "(TagName=\"Gameplay.Status.StormSickness\")")
            (Utilities|CheckBroadcastandCamera
              (:Show
                (CallFunction|AddSceneifUnactive)))))))
    (:Added
      (CallFunction|RemoveSceneifActive))))

(event Custom|OnFortPlayerDied
  (params (FortPlayerPawnAthena :type Fort%20Player%20Pawn%20Athena%20Object%20Reference))
  ;[1 1 1] Death
  (DynamicUIDirector|RemoveScene "/StormSickness/UI/StormSickness_Scene.StormSickness_Scene"))

(fn AddSceneIfUnactive
  (if (not (Variables|Default|GetSceneActive))
    (DynamicUIDirector|AddScene "/StormSickness/UI/StormSickness_Scene.StormSickness_Scene")
    (Variables|Default|SetSceneActive true)))

(fn RemoveSceneIfActive
  (if (Variables|Default|GetSceneActive)
    (DynamicUIDirector|RemoveScene "/StormSickness/UI/StormSickness_Scene.StormSickness_Scene")
    (Variables|Default|SetSceneActive false)))

(fn ConstructionScript)
