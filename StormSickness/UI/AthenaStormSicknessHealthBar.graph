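# Storm Sickness health-bar HUD overlay widget, built on AthenaStormSurgeWarningBase, that shows faint storm-surge glow and lightning effects.
# Layered low-opacity material brushes form the visual, and Anim_Load plays StormDMGAlert, which animates LightningBack's horizontal scale.
# StartStormSurgeDisplay and StopStormSurgeDisplay are declared without logic, and the event graph is empty.
# BuildDataRegistryResolverScope supplies a resolver scope from the owning player's spatial gameplay interface name.
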WidgetBlueprint /StormSickness/UI/AthenaStormSicknessHealthBar
animations [StormDMGAlert {
  binding_count=1
  display_label=""
  length_seconds=2.5
  object=&self:StormDMGAlert
  sequence {bindings [LightningBack {
     id=#1
     possessed_class=Image
     tracks [{
       type=MovieScene2DTransformTrack
       sections [{
         type=MovieScene2DTransformSection
         end_frame=49
         start_frame=0
         channels [
          Translation.X {type=MovieSceneScriptingFloatChannel, const [0,1]}
          Translation.Y {type=MovieSceneScriptingFloatChannel, const [0,1]}
          Angle {type=MovieSceneScriptingFloatChannel, const [0,1]}
          Scale.X {type=MovieSceneScriptingFloatChannel, keys [[9,0],[16,0],[21,1]]}
          Scale.Y {type=MovieSceneScriptingFloatChannel, const [0.9,1]}
          Shear.X {type=MovieSceneScriptingFloatChannel, const [0,1]}
          Shear.Y {type=MovieSceneScriptingFloatChannel, const [0,1]}
         ]
        }]
      }]
    }]}
 }]
graphs [
 Anim_Load {nodes [
   Anim_Load() {is=title K2Node_FunctionEntry}
   > PlayAnimation(InAnimation=Get StormDMGAlert() -> (StormDMGAlert:Widget Animation Object Reference))
  ]}
 BuildDataRegistryResolverScope {nodes [
   BuildDataRegistryResolverScope() {is=title K2Node_FunctionEntry}
   n1 = GetOwningPlayer()
   > Return Node!(
    execute
    InOutResolverScopes:Array of Names=Make Array("[0]":Name=GetSpatialGameplayInterfaceFNameForActor!(Actor:Actor Object Reference=@n1) -> (ReturnValue:Name=None)) -> (Array:Array of Names)
    InOutPriority:int=1
    ReturnValue:bool=true
   )
  ]}
 EventGraph {links [], nodes []}
 GetCacheContextOverride {nodes [
   GetCacheContextOverride() {is=title K2Node_FunctionEntry}
   > Return Node!(execute, ReturnValue:Object Reference)
  ]}
 StartStormSurgeDisplay {nodes [StartStormSurgeDisplay() {is=title K2Node_FunctionEntry}]}
 StopStormSurgeDisplay {nodes [StopStormSurgeDisplay() {is=title K2Node_FunctionEntry}]}
]
variables [
 StartStormSurgeDisplay {}
 StopStormSurgeDisplay {}
]
widgets {
 info {inheritedWidgetCount=1, parentClass=/Script/FortniteUI.AthenaStormSurgeWarningBase, rootWidgetClass=&SizeBox, widgetCount=9}
 widgets [
  SizeBox_0 {widgetClassPath=&SizeBox, widget=SizeBox_0}
  VerticalBox_2 {
   widgetClassPath=&VerticalBox
   parent=SizeBox_0
   widget=VerticalBox_2
   slot {class=SizeBoxSlot, properties {horizontalAlignment=HAlign_Center, verticalAlignment=VAlign_Center}}
  }
  Overlay_3 {
   widgetClassPath=&Overlay
   parent=VerticalBox_2
   widget=Overlay_3
   slot {class=VerticalBoxSlot, properties {size {sizeRule=Fill, value=1}}}
  }
  Border_595 {
   widgetClassPath=&Border
   parent=Overlay_3
   widget=Border_595
   properties {
    renderOpacity=0.0625
    background {
     bIsDynamicallyLoaded=false
     drawAs=Box
     imageType=NoImage
     mirroring=NoMirror
     resourceName=None
     resourceObject=/Game/Athena/HUD/LTM/Wax/Materials/M_UI_TextGlow_Add_Gradient_StormSurge
     tiling=NoTile
     imageSize=<128 128>
     margin=<ltrb 0.5 0.5 0.5 0.5>
     outlineSettings {
      bUseBrushTransparency=false
      roundingType=HalfHeightRadius
      width=0
      color {colorUseRule=UseColor_Specified, specifiedColor=<rgba 0 0 0 0>}
      cornerRadii=<0 0 0 0>
     }
     tintColor {colorUseRule=UseColor_Specified, specifiedColor=<rgba 1 1 1 1>}
     uVRegion {bIsValid=false, max=<0 0>, min=<0 0>}
    }
    padding=<ltrb 0 0 0 0>
    renderTransform {angle=0, scale=<1.25 2.5>, shear=<0 0>, translation=<0 0>}
   }
   slot {class=OverlaySlot, properties {horizontalAlignment=HAlign_Fill, verticalAlignment=VAlign_Fill}}
  }
  LightningBack {
   widgetClassPath=&Image
   parent=Overlay_3
   widget=LightningBack
   properties {
    renderOpacity=0.125
    brush {
     bIsDynamicallyLoaded=false
     drawAs=Image
     imageType=NoImage
     mirroring=NoMirror
     resourceName=None
     resourceObject=/Game/Athena/HUD/Materials/M_UI_LightningFX_Inst
     tiling=NoTile
     imageSize=<375 64>
     margin=<ltrb 0 0 0 0>
     outlineSettings {
      bUseBrushTransparency=false
      roundingType=HalfHeightRadius
      width=0
      color {colorUseRule=UseColor_Specified, specifiedColor=<rgba 0 0 0 0>}
      cornerRadii=<0 0 0 0>
     }
     tintColor {colorUseRule=UseColor_Specified, specifiedColor=<rgba 1 1 1 1>}
     uVRegion {bIsValid=false, max=<0 0>, min=<0 0>}
    }
    renderTransform {angle=0, scale=<1 0.9>, shear=<0 0>, translation=<0 0>}
   }
   slot {class=OverlaySlot, properties {verticalAlignment=VAlign_Bottom}}
  }
  HorizontalBox_95 {
   widgetClassPath=&HorizontalBox
   parent=Overlay_3
   widget=HorizontalBox_95
   properties {renderOpacity=0.0625}
   slot {class=OverlaySlot, properties {horizontalAlignment=HAlign_Fill, verticalAlignment=VAlign_Fill}}
  }
  LightningWrap1 {
   widgetClassPath=&Image
   parent=HorizontalBox_95
   widget=LightningWrap1
   properties {
    brush {
     bIsDynamicallyLoaded=false
     drawAs=Image
     imageType=NoImage
     mirroring=NoMirror
     resourceName=None
     resourceObject=/Game/Athena/HUD/XP/Materials/M_UI_PwrUp_Lightning
     tiling=NoTile
     imageSize=<128 32>
     margin=<ltrb 0 0 0 0>
     outlineSettings {
      bUseBrushTransparency=false
      roundingType=HalfHeightRadius
      width=0
      color {colorUseRule=UseColor_Specified, specifiedColor=<rgba 0 0 0 0>}
      cornerRadii=<0 0 0 0>
     }
     tintColor {colorUseRule=UseColor_Specified, specifiedColor=<rgba 1 1 1 1>}
     uVRegion {bIsValid=false, max=<0 0>, min=<0 0>}
    }
    renderTransform {angle=0, scale=<1 2>, shear=<0 0>, translation=<0 0>}
   }
   slot {class=HorizontalBoxSlot, properties {padding=<ltrb 40 0 0 0>}}
  }
  LightningWrap3 {
   widgetClassPath=&Image
   parent=HorizontalBox_95
   widget=LightningWrap3
   properties {
    brush {
     bIsDynamicallyLoaded=false
     drawAs=Image
     imageType=NoImage
     mirroring=NoMirror
     resourceName=None
     resourceObject=/Game/Athena/HUD/XP/Materials/M_UI_PwrUp_Lightning_3
     tiling=NoTile
     imageSize=<128 32>
     margin=<ltrb 0 0 0 0>
     outlineSettings {
      bUseBrushTransparency=false
      roundingType=HalfHeightRadius
      width=0
      color {colorUseRule=UseColor_Specified, specifiedColor=<rgba 0 0 0 0>}
      cornerRadii=<0 0 0 0>
     }
     tintColor {colorUseRule=UseColor_Specified, specifiedColor=<rgba 1 1 1 1>}
     uVRegion {bIsValid=false, max=<0 0>, min=<0 0>}
    }
    renderTransform {angle=0, scale=<1 -2>, shear=<0 0>, translation=<0 0>}
   }
   slot {class=HorizontalBoxSlot, properties {horizontalAlignment=HAlign_Right, padding=<ltrb 0 0 48 0>, size {sizeRule=Fill, value=1}}}
  }
  TutorialHighlight {widgetClassPath=&TutorialHighlight}
 ]
}
