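# Touch UI widget wrapping the storm sickness health bar, showing it while the tracked pawn has storm sickness and is outside the safe zone.
# Async gameplay-tag listeners on the cached pawn toggle visibility, and player death collapses the bar.
# View target changes to another pawn re-cache it and start new listeners without explicitly ending old ones.
# A macro blocks showing the bar for live broadcast controllers in free drone camera.
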WidgetBlueprint /StormSickness/UI/TouchStormSicknessHealth
graphs [
 CheckBroadcastAndCamera {nodes [
   Inputs()
   > n1 = Cast To FortLiveBroadcastController!(execute, Object=GetOwningPlayer()) -> (then, CastFailed, AsFort Live Broadcast Controller:Fort Live Broadcast Controller Object Reference) {is=title K2Node_DynamicCast}
   > n2 = Branch(Condition="Not Equal (Enum)"!(
     A:ESpectatorCameraType Enum=Get CurrentCameraType!(self:Fort Player Controller Spectating Object Reference=@n1) -> (CurrentCameraType:ESpectatorCameraType Enum=ThirdPerson)
     B:ESpectatorCameraType Enum=DroneFree
    ) -> (ReturnValue:bool))
   Outputs!(Show:Exec=[@n2.then, @n1.CastFailed], Hide:Exec=@n2.else)
  ]}
 EventGraph {nodes [
   ListenToPawnTags() {is=title K2Node_CustomEvent}
   > n1 = WaitGameplayTagAddToActor!(
    execute:Exec
    TargetActor:Actor Object Reference=Get Cached Player Pawn() -> (Cached Player Pawn:Fort Player Pawn Athena Object Reference)
    Tag:Gameplay Tag Structure=(TagName="Gameplay.InsideSafeZone")
    OnlyTriggerOnce:bool=false
   ) -> (then:Exec, AsyncTaskProxy:Ability Async Wait Gameplay Tag Added Object Reference, Added:Exec) {is=title K2Node_AsyncAction}
   n2 = Get Cached Player Pawn()
   > n3 = WaitGameplayTagRemoveFromActor!(
    execute:Exec
    TargetActor:Actor Object Reference=@n2
    Tag:Gameplay Tag Structure=(TagName="Gameplay.InsideSafeZone")
    OnlyTriggerOnce:bool=false
   ) -> (then:Exec, AsyncTaskProxy:Ability Async Wait Gameplay Tag Removed Object Reference, Removed:Exec) {is=title K2Node_AsyncAction}
   n4 = Get Cached Player Pawn()
   > n5 = WaitGameplayTagAddToActor(TargetActor=@n4, Tag=(TagName="Gameplay.Status.StormSickness"), OnlyTriggerOnce=false)
   > n6 = WaitGameplayTagRemoveFromActor(TargetActor=Get Cached Player Pawn(), Tag=(TagName="Gameplay.Status.StormSickness"), OnlyTriggerOnce=false)
   n7 = OnFortPlayerDied!() -> (OutputDelegate, then, FortPlayerPawnAthena:Fort Player Pawn Athena Object Reference) {is=title K2Node_CustomEvent}
   Bind Event to On Fort Player Died(execute=@n6.then, self:Fort Player Pawn Athena Object Reference=Get Cached Player Pawn(), Delegate=@n7) {is=title K2Node_AddDelegate}
   SetVisibility(
    execute=@n6.Removed
    self=Get AthenaStormSicknessHealthBar() -> (AthenaStormSicknessHealthBar:Athena Storm Sickness Health Bar Object Reference)
    InVisibility=Collapsed
   )
   n8 = Branch(
    execute=@n5.Added
    Condition=HasMatchingGameplayTag!(self:Gameplay Tag Asset Interface Interface=@n4, TagToCheck:Gameplay Tag Structure=(TagName="Gameplay.InsideSafeZone")) -> (ReturnValue:bool)
   )
   Check Broadcast and Camera!(execute:Exec=@n8.else) -> (Show:Exec, Hide:Exec) {is=title K2Node_MacroInstance}
   > SetVisibility(self=Get AthenaStormSicknessHealthBar(), InVisibility=HitTestInvisible)
   Branch(execute=@n3.Removed, Condition=HasMatchingGameplayTag(self=@n2, TagToCheck=(TagName="Gameplay.Status.StormSickness")))
   > Check Broadcast and Camera()
   > SetVisibility(self=Get AthenaStormSicknessHealthBar(), InVisibility=HitTestInvisible)
   SetVisibility(execute=@n1.Added, self=Get AthenaStormSicknessHealthBar(), InVisibility=Collapsed)
   n9 = ViewTargetChanged!() -> (
    OutputDelegate
    then
    PlayerController:Fort Player Controller Object Reference
    OldViewTarget:Actor Object Reference
    NewViewTarget:Actor Object Reference
   ) {is=title K2Node_CustomEvent}
   > n10 = Cast To FortPlayerPawnAthena(Object=@n9.NewViewTarget)
   > Set Cached Player Pawn(Cached Player Pawn:Fort Player Pawn Athena Object Reference=@n10) -> (Output_Get:Fort Player Pawn Athena Object Reference)
   > ListenToPawnTags() {is=title K2Node_CallFunction}
   Check Broadcast and Camera(execute=@n10.CastFailed)
   > SetVisibility(self=Get AthenaStormSicknessHealthBar(), InVisibility=Collapsed)
   Event Construct()
   > n11 = Cast To FortPlayerController(Object=GetOwningPlayer())
   > n12 = Cast To FortPlayerPawnAthena(Object=GetPlayerPawn!(self:Fort Player Controller Object Reference=@n11) -> (ReturnValue:Fort Player Pawn Object Reference))
   > Set Cached Player Pawn(Cached Player Pawn=@n12)
   > n13 = ListenToPawnTags()
   Bind Event to On View Target Changed(execute=[@n13.then, @n12.CastFailed], self:Fort Player Controller Object Reference=@n11, Delegate=@n9) {is=title K2Node_AddDelegate}
   SetVisibility(execute=@n7.then, self=Get AthenaStormSicknessHealthBar(), InVisibility=Collapsed)
  ]}
]
variables [Cached Player Pawn {type=FortPlayerPawnAthena}]
widgets {
 info {
  inheritedWidgetCount=0
  parentClass=&UserWidget
  rootWidgetClass=&/AthenaStormSicknessHealthBar.AthenaStormSicknessHealthBar_C
  widgetCount=1
 }
 widgets [AthenaStormSicknessHealthBar {widgetClassPath=&/AthenaStormSicknessHealthBar, bIsVariable=true, widget=AthenaStormSicknessHealthBar}]
}
